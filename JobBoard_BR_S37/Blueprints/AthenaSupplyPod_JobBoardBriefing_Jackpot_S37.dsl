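; Job Board briefing supply pod (Jackpot, S37) that detects when a player outside the owner's team poaches it.
; On interact it runs the parent logic, then compares the interacting pawn's team with the pod's owner.
; If poached, it gathers the owner and squadmates, applies the poached gameplay effect to each, and broadcasts a tagged CustomVerb.JobBoard.Poached message per player for quest progress.
; It also prints an on-screen debug string when poached.
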(event Interaction|EventBlueprintOnInteract
  (params (InteractingPawn :type Fort%20Pawn%20Object%20Reference) (InteractionBeingAttempted :type EInteractionBeingAttempted%20Enum :default "FirstInteraction"))
  ;[0 0 0] On interact - check to see if this supply drop was poached by an enemy team
  (|CallParentFunction InteractingPawn InteractionBeingAttempted)
  (if (not (Team|OnSameTeam InteractingPawn (Actor|GetOwner)))
    ;;[0.2396 0.2396 0.2396] Poached - Broadcast a custom verb message to complete an objective on the quest
    (Development|PrintString "Poached by enemy player" true true "(R=1.000000,G=0.000000,B=0.573688,A=1.000000)" 4.0)
    (bind _fortplayerstates (CallFunction|GetOwnerandSquadmates))
    (for _array_element _fortplayerstates
      (GameplayEffects|ApplyGameplayEffectToSelf :self (Ability|GetFortAbilitySystemComponent (PlayerState|GetPawn _array_element)) :GameplayEffectClass (Variables|Default|GetPoachedGameplayEffect) :Level 0.0)
      (bind _result (CustomVerb|AddTagsforSubject (CustomVerb|MakeCustomVerbMessage (PlayerState|GetPawn _array_element) "(TagName=\"CustomVerb.JobBoard.Poached\")") "Job Type" (Variables|Default|GetPoachedSubjectTags)))
      (CustomVerb|BroadcastCustomVerbMessage _result))))

(fn GetOwnerAndSquadmates
  (bind _returnvalue (Actor|GetOwner))
  ;[0 0 0] Get the owner of this actor and their squadmates who are participating in the quest
  (bind _asfort_player_controller (Utilities|Casting|CastToFortPlayerController _returnvalue))
  (bind _returnvalue_2 (Fortnite|Utilities|AttemptGetPlayerStateforPawn (Game|Player|GetPlayerPawn _asfort_player_controller)))
  (bind _asfort_player_state_athena (Utilities|Casting|CastToFortPlayerStateAthena _returnvalue_2))
  (bind _returnvalue_3 (Class|FortGameStateAthena|GetSquadMemberPlayerStates (Fortnite|Athena|GameState|GetGameStateAthenafromContext _returnvalue) _asfort_player_state_athena true))
  (return _returnvalue_3))

(fn ConstructionScript
  (|CallParentFunction))
