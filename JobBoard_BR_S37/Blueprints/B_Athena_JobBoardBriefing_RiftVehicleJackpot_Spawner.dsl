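; Spawner for the Job Board "Rift Vehicle Jackpot" briefing that creates a Choppa (Hoagie) vehicle.
; SpawnVehicle takes an entry from the SpawnVehicleClass array and, if the class is valid, spawns it at ChosenQueryLocation, adjusting for collision but always spawning.
; It then casts the result to FortHoagieVehicle, stores it in VehicleTypeSpawned, wakes it and sets its fuel.
; The event graph is empty, and nothing in this file calls SpawnVehicle.
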(event EventBeginPlay
  (|CallParentFunction) ; This node is disabled and will not be called.\nDrag off pins to build functionality.
)

(event Collision|EventActorBeginOverlap
  (params (OtherActor :type Actor%20Object%20Reference)))

(event EventTick
  (params (DeltaSeconds :type Float%20%28single-precision%29)))

(fn SpawnVehicle
  (bind _spawnvehicleclass (Variables|Default|GetSpawnVehicleClass))
  (bind _output (Utilities|Array|Get(acopy) _spawnvehicleclass))
  ;[0.1927 0.1927 0.1927 0.8] Get the Choppa vehicle and spawn it
  (if (Utilities|IsValidClass _output)
    (bind _returnvalue (Game|SpawnActorfromClass _output (Math|Conversions|ToTransform(Vector) (Variables|Default|GetChosenQueryLocation)) "AdjustIfPossibleButAlwaysSpawn" "SelectDefaultAtRuntime" 0 false 0))
    ;[0.1927 0.1927 0.1927 0.8] Set up vehicle initial values
    (bind _asfort_hoagie_vehicle (Utilities|Casting|CastToFortHoagieVehicle _returnvalue))
    (Variables|Default|SetVehicleTypeSpawned _asfort_hoagie_vehicle)
    (Vehicle|WakeUp _asfort_hoagie_vehicle)
    (Vehicle|Fuel|SetFuel _asfort_hoagie_vehicle 1e+02)))

(fn ConstructionScript
  (|CallParentFunction))
