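# Ambient loop sound cue for the job board's Shadow Briefing jackpot, in the loot ambient sound class and concurrency set.
# Each mix pairs a randomly picked looping bed using chest attenuation with half-volume random pings that repeat after random delays with pitch and volume variation.
# The pings use job board mono attenuation with a logarithmic falloff override.
# A quality switch gives the two higher slots a leaner mix with fewer bed and ping variants.
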SoundCue /JobBoard_BR_S37/Sounds/SC_JobBoard_ShadowBriefing_Jackpot_Ambient_Loop
root=QualityLevel_0
nodes [
 {id=Attenuation_0, children [Random_0], properties {attenuationSettings=/Game/Sounds/Attenuation/Fort_Chests_Att}}
 {
  id=Attenuation_1
  children [Looping_0]
  properties {
   attenuationSettings=/JobBoard/Sounds/Fort_JobBoard_Ambient_Loop_Mono_Default_Att
   attenuationOverrides {distanceAlgorithm=Logarithmic, falloffDistance=3000}
  }
 }
 {id=Attenuation_2, children [Random_2], properties {attenuationSettings=&Fort_Chests_Att}}
 {
  id=Attenuation_3
  children [Looping_1]
  properties {
   attenuationSettings=&Fort_JobBoard_Ambient_Loop_Mono_Default_Att
   attenuationOverrides {distanceAlgorithm=Logarithmic, falloffDistance=3000}
  }
 }
 {id=Delay_0, children [Modulator_0], properties {delayMax=7, delayMin=2}}
 {id=Delay_1, children [Modulator_1], properties {delayMax=7, delayMin=2}}
 {id=Looping_0, children [Delay_0]}
 {id=Looping_1, children [Delay_1]}
 {id=Mixer_0, children [Attenuation_0, Attenuation_1], properties {inputVolume [1,0.5]}}
 {id=Mixer_1, children [Attenuation_2, Attenuation_3], properties {inputVolume [1,0.5]}}
 {id=Modulator_0, children [Random_3], properties {pitchMax=1.2, pitchMin=1.1, volumeMax=1, volumeMin=0.85}}
 {id=Modulator_1, children [Random_1], properties {pitchMax=1.2, pitchMin=1.1, volumeMax=1, volumeMin=0.85}}
 {id=QualityLevel_0, children [Mixer_0,Mixer_0,Mixer_1,Mixer_1]}
 {id=Random_0, children [WavePlayer_0,WavePlayer_1,WavePlayer_2,WavePlayer_3], properties {weights [1,1,1,1]}}
 {id=Random_1, children [WavePlayer_4,WavePlayer_5,WavePlayer_6], properties {weights [1,1,1]}}
 {id=Random_2, children [WavePlayer_9, WavePlayer_15], properties {weights [1,1]}}
 {id=Random_3, children [WavePlayer_10, WavePlayer_11, WavePlayer_12, WavePlayer_13, WavePlayer_14], properties {weights [1,1,1,1,1]}}
 {id=WavePlayer_0, properties {bLooping=true, soundWaveAssetPtr=&/JobBoard_ShadowBriefing_Jackpot_Amb_Loop_04}}
 {id=WavePlayer_1, properties {bLooping=true, soundWaveAssetPtr=&/JobBoard_ShadowBriefing_Jackpot_Amb_Loop_03}}
 {id=WavePlayer_10, properties {soundWaveAssetPtr=&/JobBoard_ShadowBriefing_Jackpot_Amb_Ping_01}}
 {id=WavePlayer_11, properties {soundWaveAssetPtr=&/JobBoard_ShadowBriefing_Jackpot_Amb_Ping_02}}
 {id=WavePlayer_12, properties {soundWaveAssetPtr=&/JobBoard_ShadowBriefing_Jackpot_Amb_Ping_03}}
 {id=WavePlayer_13, properties {soundWaveAssetPtr=&/JobBoard_ShadowBriefing_Jackpot_Amb_Ping_04}}
 {id=WavePlayer_14, properties {soundWaveAssetPtr=&/JobBoard_ShadowBriefing_Jackpot_Amb_Ping_05}}
 {id=WavePlayer_15, properties {bLooping=true, soundWaveAssetPtr=&JobBoard_ShadowBriefing_Jackpot_Amb_Loop_03}}
 {id=WavePlayer_2, properties {bLooping=true, soundWaveAssetPtr=&/JobBoard_ShadowBriefing_Jackpot_Amb_Loop_02}}
 {id=WavePlayer_3, properties {bLooping=true, soundWaveAssetPtr=&/JobBoard_ShadowBriefing_Jackpot_Amb_Loop_01}}
 {id=WavePlayer_4, properties {soundWaveAssetPtr=&JobBoard_ShadowBriefing_Jackpot_Amb_Ping_01}}
 {id=WavePlayer_5, properties {soundWaveAssetPtr=&JobBoard_ShadowBriefing_Jackpot_Amb_Ping_02}}
 {id=WavePlayer_6, properties {soundWaveAssetPtr=&JobBoard_ShadowBriefing_Jackpot_Amb_Ping_03}}
 {id=WavePlayer_9, properties {bLooping=true, soundWaveAssetPtr=&JobBoard_ShadowBriefing_Jackpot_Amb_Loop_04}}
]
properties {
 maxDistance=3500
 soundClassObject=/Game/Sounds/FortSoundClassesAndModes/Fort_Loot_Ambient_Loop
 volumeMultiplier=0.3
 concurrencyOverrides {maxCount=2}
 concurrencySet [/Game/Sounds/Concurrency_Rules/Loot_Ambient_Conc]
}
