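# Niagara system that, by its name, provides an idle effect for a Jackpot element in the Job Board content.
# It contains an enabled Glow emitter and a disabled Sparkles001 emitter, with no renderers listed for either.
# SystemState is the only listed module, and the system exposes no user variables, data interfaces or dynamic inputs.
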NiagaraSystem /JobBoard_BR_S37/VFX/NS_LWE_Jackpot_Idle
dependencies {usedDataInterfaces [], usedDynamicInputs [], usedModules [&SystemState]}
emitters {Glow {rendererClasses [], renderers []}, Sparkles001 {bEnabled=false, rendererClasses [], renderers []}}
summary {
 systemName=NS_LWE_Jackpot_Idle
 emitters [{bEnabled=false, emitterName=Sparkles001, rendererClasses []}, {emitterName=Glow, rendererClasses []}]
 userVariables []
}
